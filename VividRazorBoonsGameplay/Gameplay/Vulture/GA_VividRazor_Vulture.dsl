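; Vulture boon gameplay ability that lets the owning player see elimination markers by overriding their detection range.
; On activation, where cosmetic events cannot execute, it sets the scalable-float range on the controller's ViewEliminations component.
; A non-negative duration starts a timer that ends the ability; a negative duration starts none.
; On end it clears the timer, then on that same non-cosmetic path disables the override and removes the granting effect.
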(event Custom|Boon%20Duration%20Expired
  ;[1 1 1] Set Boon Duration Before ending detection.
  (Ability|EndAbility))

(event Ability|EventActivateAbility
  (bind _hfboonduration (Variables|Default|GetHFBoonDuration))
  (bind _returnvalue (ScalableFloat|GetValueatLevel _hfboonduration))
  (Networking|CanExecuteCosmeticEvents
    (:True
      _)
    (:False
      ;[1 1 1] Allow seeing elimination markers. Range ==0 is infinite range.<0 is off.
      (bind _asfort_player_controller_athena (Utilities|Casting|CastToFortPlayerControllerAthena (Ability|GetPlayerControllerfromActorInfo))
        (:then
          (Utilities|IsValid (Actor|GetComponentbyClass _asfort_player_controller_athena "/Script/FortniteGame.FortControllerComponent_ViewEliminations")
            (:Is%20Valid
              (bind _output_get (Variables|Default|SetTrackedViewEliminationsComponent (Actor|GetComponentbyClass _asfort_player_controller_athena "/Script/FortniteGame.FortControllerComponent_ViewEliminations")))
              (Class|FortControllerComponentViewEliminations|SetOverrideEliminationDetectionRangeonServer _output_get (ScalableFloat|GetValueatLevel (Variables|Default|GetHFElimDetectionRange)))
              (if (>= _returnvalue 0)
                ;[1 1 1] Set Boon Duration Before ending detection.
                (bind _returnvalue_2 (Utilities|Time|SetTimerbyEvent (delegate Custom|Boon%20Duration%20Expired) _returnvalue))
                (Variables|Default|SetBoonDurationTimerHandle _returnvalue_2)
                (else
                  _))
              (|CallParentFunction))
            (:Is%20Not%20Valid
              _)))
        (:CastFailed
          _)))))

(event Ability|EventOnEndAbility
  (params (bWasCancelled :type Boolean))
  (Utilities|Time|ClearandInvalidateTimerbyHandle (Variables|Default|GetBoonDurationTimerHandle))
  (Networking|CanExecuteCosmeticEvents
    (:True
      _)
    (:False
      (bind _trackedvieweliminationscomponent (Variables|Default|GetTrackedViewEliminationsComponent)
        (:then
          (Class|FortControllerComponentViewEliminations|DisableOverrideEliminationDetectionRangeonServer _trackedvieweliminationscomponent))
        (:else
          _))
      (Ability|RemoveGrantedbyEffect)
      (|CallParentFunction bWasCancelled))))
